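# Control Rig for the SuperNovaTaro face accessory that applies pose-driven corrective offsets to short hair and crown bones.
# Spherical pose readers on head and right upper arm write alphas, some multiplied into combined look-and-arm weights for Modify Transforms nodes.
# A Crouch bool selects separately tuned branches; crouch adds a spine_05-relative look-up reader, pistol/rifle blends and a full-weight crown offset.
# Left/right look, skydiving and look-down-arm-back alphas are computed but drive nothing here.
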ControlRigBlueprint /BRCosmetics/Characters/Player/Female/Medium/Bodies/F_MED_SuperNovaTaro/Meshes/Parts/F_MED_SuperNovaTaro_FaceAcc_CtrlRig
graphs [
 ContainedGraph {nodes [
   n1 = RigVMFunctionEntryNode()
   SequenceExecution_4 = &RigVMFunction_Sequence(ExecuteContext=@n1.ExecuteContext)
   SequenceExecution_4_1 = &RigVMFunction_Sequence(ExecuteContext=@SequenceExecution_4.B)
   SequenceExecution_4_2 = &RigVMFunction_Sequence(ExecuteContext=@SequenceExecution_4_1.B)
   RigVMFunctionReturnNode(A=@SequenceExecution_4.A, B=@SequenceExecution_4_1.A, C=@SequenceExecution_4_2.A, D=@SequenceExecution_4_2.B)
  ]}
 ContainedGraph {nodes [
   n1 = RigVMFunctionEntryNode()
   SequenceExecution_3 = &RigVMFunction_Sequence(ExecuteContext=@n1.ExecuteContext)
   SequenceExecution_3_1 = &RigVMFunction_Sequence(ExecuteContext=@SequenceExecution_3.B)
   SequenceExecution_3_2 = &RigVMFunction_Sequence(ExecuteContext=@SequenceExecution_3_1.B)
   SequenceExecution_3_3 = &RigVMFunction_Sequence(ExecuteContext=@SequenceExecution_3_2.B)
   RigVMFunctionReturnNode(A=@SequenceExecution_3.A, B=@SequenceExecution_3_1.A, C=@SequenceExecution_3_2.A, D=@SequenceExecution_3_3.A, E=@SequenceExecution_3_3.B)
  ]}
 ContainedGraph {nodes [
   n1 = RigVMFunctionEntryNode()
   SequenceExecution_2 = &RigVMFunction_Sequence(ExecuteContext=@n1.ExecuteContext)
   SequenceExecution_2_1 = &RigVMFunction_Sequence(ExecuteContext=@SequenceExecution_2.B)
   SequenceExecution_2_2 = &RigVMFunction_Sequence(ExecuteContext=@SequenceExecution_2_1.B)
   RigVMFunctionReturnNode(A=@SequenceExecution_2.A, B=@SequenceExecution_2_1.A, C=@SequenceExecution_2_2.A, D=@SequenceExecution_2_2.B)
  ]}
 ContainedGraph {nodes [
   n1 = RigVMFunctionEntryNode()
   SequenceExecution_7 = &RigVMFunction_Sequence(ExecuteContext=@n1.ExecuteContext)
   SequenceExecution_7_1 = &RigVMFunction_Sequence(ExecuteContext=@SequenceExecution_7.B)
   SequenceExecution_7_2 = &RigVMFunction_Sequence(ExecuteContext=@SequenceExecution_7_1.B)
   RigVMFunctionReturnNode(A=@SequenceExecution_7.A, B=@SequenceExecution_7_1.A, C=@SequenceExecution_7_2.A, D=@SequenceExecution_7_2.B)
  ]}
 ContainedGraph {nodes [
   n1 = RigVMFunctionEntryNode()
   SequenceExecution_12 = &RigVMFunction_Sequence(ExecuteContext=@n1.ExecuteContext)
   SequenceExecution_12_1 = &RigVMFunction_Sequence(ExecuteContext=@SequenceExecution_12.B)
   RigVMFunctionReturnNode(A=@SequenceExecution_12.A, B=@SequenceExecution_12_1.A, C=@SequenceExecution_12_1.B)
  ]}
 ContainedGraph {nodes [
   n1 = RigVMFunctionEntryNode()
   SequenceExecution_4 = &RigVMFunction_Sequence(ExecuteContext=@n1.ExecuteContext)
   SequenceExecution_4_1 = &RigVMFunction_Sequence(ExecuteContext=@SequenceExecution_4.B)
   SequenceExecution_4_2 = &RigVMFunction_Sequence(ExecuteContext=@SequenceExecution_4_1.B)
   RigVMFunctionReturnNode(A=@SequenceExecution_4.A, B=@SequenceExecution_4_1.A, C=@SequenceExecution_4_2.A, D=@SequenceExecution_4_2.B)
  ]}
 ContainedGraph {nodes [
   n1 = RigVMFunctionEntryNode()
   SequenceExecution_3 = &RigVMFunction_Sequence(ExecuteContext=@n1.ExecuteContext)
   SequenceExecution_3_1 = &RigVMFunction_Sequence(ExecuteContext=@SequenceExecution_3.B)
   SequenceExecution_3_2 = &RigVMFunction_Sequence(ExecuteContext=@SequenceExecution_3_1.B)
   SequenceExecution_3_3 = &RigVMFunction_Sequence(ExecuteContext=@SequenceExecution_3_2.B)
   SequenceExecution_3_4 = &RigVMFunction_Sequence(ExecuteContext=@SequenceExecution_3_3.B)
   SequenceExecution_3_5 = &RigVMFunction_Sequence(ExecuteContext=@SequenceExecution_3_4.B)
   RigVMFunctionReturnNode(
    A=@SequenceExecution_3.A
    B=@SequenceExecution_3_1.A
    C=@SequenceExecution_3_2.A
    D=@SequenceExecution_3_3.A
    E=@SequenceExecution_3_4.A
    F=@SequenceExecution_3_5.A
    G=@SequenceExecution_3_5.B
   )
  ]}
 ContainedGraph {nodes [
   n1 = RigVMFunctionEntryNode()
   SequenceExecution_2 = &RigVMFunction_Sequence(ExecuteContext=@n1.ExecuteContext)
   SequenceExecution_2_1 = &RigVMFunction_Sequence(ExecuteContext=@SequenceExecution_2.B)
   SequenceExecution_2_2 = &RigVMFunction_Sequence(ExecuteContext=@SequenceExecution_2_1.B)
   RigVMFunctionReturnNode(A=@SequenceExecution_2.A, B=@SequenceExecution_2_1.A, C=@SequenceExecution_2_2.A, D=@SequenceExecution_2_2.B)
  ]}
 ContainedGraph {nodes [
   n1 = RigVMFunctionEntryNode()
   SequenceExecution_7 = &RigVMFunction_Sequence(ExecuteContext=@n1.ExecuteContext)
   SequenceExecution_7_1 = &RigVMFunction_Sequence(ExecuteContext=@SequenceExecution_7.B)
   SequenceExecution_7_2 = &RigVMFunction_Sequence(ExecuteContext=@SequenceExecution_7_1.B)
   RigVMFunctionReturnNode(A=@SequenceExecution_7.A, B=@SequenceExecution_7_1.A, C=@SequenceExecution_7_2.A, D=@SequenceExecution_7_2.B)
  ]}
 ContainedGraph {nodes [
   n1 = RigVMFunctionEntryNode()
   SequenceExecution_12 = &RigVMFunction_Sequence(ExecuteContext=@n1.ExecuteContext)
   SequenceExecution_12_1 = &RigVMFunction_Sequence(ExecuteContext=@SequenceExecution_12.B)
   SequenceExecution_12_2 = &RigVMFunction_Sequence(ExecuteContext=@SequenceExecution_12_1.B)
   RigVMFunctionReturnNode(A=@SequenceExecution_12.A, B=@SequenceExecution_12_1.A, C=@SequenceExecution_12_2.A, D=@SequenceExecution_12_2.B)
  ]}
 RigVMFunctionLibrary {links [], nodes []}
 RigVMModel {nodes [
   RigUnit_BeginExecution = &RigUnit_BeginExecution()
   RigVMFunction_ControlFlowBranch = &RigVMFunction_ControlFlowBranch(ExecuteContext=@RigUnit_BeginExecution.ExecutePin, Condition=RigVMVariableNode() {variable=Crouch}.Value)
   SequenceExecution_12_1 = RigVMAggregateNode(ExecuteContext=@RigVMFunction_ControlFlowBranch.False)
   SequenceExecution_3 = RigVMAggregateNode(ExecuteContext=@SequenceExecution_12_1.A)
   SequenceExecution_4 = RigVMAggregateNode(ExecuteContext=@SequenceExecution_3.A)
   SequenceExecution_7 = RigVMAggregateNode(ExecuteContext=@SequenceExecution_4.A)
   n1 = &RigUnit_SphericalPoseReader(
    ActiveRegionScaleFactors=(PositiveWidth=1,NegativeWidth=1,PositiveHeight=1,NegativeHeight=1)
    ActiveRegionSize=0.065
    Debug=(bDrawDebug=false,DebugScale=25,DebugSegments=20,DebugThickness=0.25)
    DriverAxis=(X=-1,Y=0,Z=0)
    DriverItem=(Type=Bone,Name="upperarm_r")
    FalloffRegionScaleFactors=(PositiveWidth=1,NegativeWidth=1,PositiveHeight=1,NegativeHeight=1)
    FalloffSize=0.125
    OptionalParentItem=(Type=Bone)
    RotationOffset=(X=80,Y=0,Z=60)
    ExecutePin=@SequenceExecution_7.A
   )
   > RigVMVariableNode(Value=@n1.OutputParam) {variable=upperarm_r_SPR_back_alpha}
   n2 = &RigUnit_SphericalPoseReader(
    ActiveRegionScaleFactors=(PositiveWidth=1,NegativeWidth=1,PositiveHeight=1,NegativeHeight=1)
    ActiveRegionSize=0.075
    Debug=(bDrawDebug=false,DebugScale=25,DebugSegments=20,DebugThickness=0.25)
    DriverAxis=(X=-1,Y=0,Z=0)
    DriverItem=(Type=Bone,Name="upperarm_r")
    FalloffRegionScaleFactors=(PositiveWidth=1,NegativeWidth=1,PositiveHeight=1,NegativeHeight=1)
    FalloffSize=0.2
    OptionalParentItem=(Type=Bone)
    RotationOffset=(X=-55,Y=0,Z=-20)
    ExecutePin=@SequenceExecution_7.D
   )
   > RigVMVariableNode(Value=@n2.OutputParam) {variable=upperarm_r_rifle_up_alpha}
   n3 = &RigUnit_SphericalPoseReader(
    ActiveRegionScaleFactors=(PositiveWidth=1,NegativeWidth=1.8,PositiveHeight=2,NegativeHeight=1)
    ActiveRegionSize=0.1
    Debug=(bDrawDebug=true,DebugScale=25,DebugSegments=20,DebugThickness=0.25)
    DriverAxis=(X=-1,Y=0,Z=0)
    DriverItem=(Type=Bone,Name="upperarm_r")
    FalloffRegionScaleFactors=(PositiveWidth=1,NegativeWidth=1.8,PositiveHeight=1.8,NegativeHeight=1)
    FalloffSize=0.25
    OptionalParentItem=(Type=Bone)
    RotationOffset=(X=-115,Y=0,Z=25)
    ExecutePin=@SequenceExecution_7.B
   )
   > RigVMVariableNode(Value=@n3.OutputParam) {variable=uppearm_r_SPR_aim_up_two_hands_alpha}
   n4 = &RigUnit_SphericalPoseReader(
    ActiveRegionScaleFactors=(PositiveWidth=2,NegativeWidth=2,PositiveHeight=1,NegativeHeight=1)
    ActiveRegionSize=0.05
    Debug=(bDrawDebug=true,DebugScale=25,DebugSegments=20,DebugThickness=0.25)
    DriverAxis=(X=-1,Y=0,Z=0)
    DriverItem=(Type=Bone,Name="upperarm_r")
    FalloffRegionScaleFactors=(PositiveWidth=1,NegativeWidth=1,PositiveHeight=0.75,NegativeHeight=0.75)
    FalloffSize=0.2
    FlipWidthScaling=false
    OptionalParentItem=(Type=Bone)
    RotationOffset=(X=5,Y=120,Z=0)
    ExecutePin=@SequenceExecution_7.C
   )
   > RigVMVariableNode(Value=@n4.OutputParam) {variable=upperarm_r_skydiving}
   SequenceExecution_2 = RigVMAggregateNode(ExecuteContext=@SequenceExecution_4.B)
   n5 = &RigUnit_SphericalPoseReader(
    ActiveRegionScaleFactors=(PositiveWidth=1,NegativeWidth=0.7,PositiveHeight=1.5,NegativeHeight=1.9)
    ActiveRegionSize=0.11
    Debug=(bDrawDebug=false,DebugScale=25,DebugSegments=20,DebugThickness=0.25)
    DriverAxis=(X=0,Y=1,Z=0)
    DriverItem=(Type=Bone,Name="head")
    FalloffRegionScaleFactors=(PositiveWidth=1,NegativeWidth=0.7,PositiveHeight=1.5,NegativeHeight=1.8)
    FalloffSize=0.25
    FlipWidthScaling=false
    OptionalParentItem=(Type=Bone)
    RotationOffset=(X=90,Y=0,Z=-25)
    ExecutePin=@SequenceExecution_2.A
   )
   > RigVMVariableNode(Value=@n5.OutputParam) {variable=head_SPR_look_up_alpha}
   n6 = &RigUnit_SphericalPoseReader(
    ActiveRegionScaleFactors=(PositiveWidth=2,NegativeWidth=0.5,PositiveHeight=2,NegativeHeight=2)
    ActiveRegionSize=0.075
    Debug=(bDrawDebug=false,DebugScale=25,DebugSegments=20,DebugThickness=0.25)
    DriverAxis=(X=0,Y=1,Z=0)
    DriverItem=(Type=Bone,Name="head")
    FalloffRegionScaleFactors=(PositiveWidth=1,NegativeWidth=0.4,PositiveHeight=1,NegativeHeight=1)
    FalloffSize=0.2
    FlipHeightScaling=false
    FlipWidthScaling=false
    OptionalParentItem=(Type=Bone)
    RotationOffset=(X=90,Y=0,Z=12.5)
    ExecutePin=@SequenceExecution_2.B
   )
   > RigVMVariableNode(Value=@n6.OutputParam) {variable=head_SPR_look_down_alpha}
   n7 = &RigUnit_SphericalPoseReader(
    ActiveRegionScaleFactors=(PositiveWidth=0.75,NegativeWidth=0.75,PositiveHeight=1,NegativeHeight=0.5)
    ActiveRegionSize=0.1
    Debug=(bDrawDebug=false,DebugScale=25,DebugSegments=20,DebugThickness=0.25)
    DriverAxis=(X=0,Y=1,Z=0)
    DriverItem=(Type=Bone,Name="head")
    FalloffRegionScaleFactors=(PositiveWidth=1,NegativeWidth=1,PositiveHeight=1,NegativeHeight=0.5)
    FalloffSize=0.2
    OptionalParentItem=(Type=Bone)
    RotationOffset=(X=120,Y=0,Z=0)
    ExecutePin=@SequenceExecution_2.C
   )
   > RigVMVariableNode(Value=@n7.OutputParam) {variable=head_SPR_look_left_alpha}
   n8 = &RigUnit_SphericalPoseReader(
    ActiveRegionScaleFactors=(PositiveWidth=0.75,NegativeWidth=0.75,PositiveHeight=1,NegativeHeight=0.5)
    ActiveRegionSize=0.1
    Debug=(bDrawDebug=false,DebugScale=25,DebugSegments=20,DebugThickness=0.25)
    DriverAxis=(X=0,Y=1,Z=0)
    DriverItem=(Type=Bone,Name="head")
    FalloffRegionScaleFactors=(PositiveWidth=1,NegativeWidth=1,PositiveHeight=1,NegativeHeight=0.5)
    FalloffSize=0.2
    FlipHeightScaling=false
    FlipWidthScaling=false
    OptionalParentItem=(Type=Bone)
    RotationOffset=(X=60,Y=0,Z=0)
    ExecutePin=@SequenceExecution_2.D
   )
   > RigVMVariableNode(Value=@n8.OutputParam) {variable=head_SPR_look_right_alpha}
   RigVMVariableNode(
    Value=&RigVMFunction_MathFloatMul(B=RigVMVariableNode() {variable=upperarm_r_SPR_back_alpha}.Value, A=RigVMVariableNode() {variable=head_SPR_look_up_alpha}.Value) {template="Multiply::Execute(in A,in B,out Result)"}.Result
    ExecuteContext=@SequenceExecution_3.E
   ) {variable=look_up_arm_back}
   RigVMVariableNode(
    Value=&RigVMFunction_MathFloatMul(B=RigVMVariableNode() {variable=uppearm_r_SPR_aim_up_two_hands_alpha}.Value, A=RigVMVariableNode() {variable=head_SPR_look_up_alpha}.Value) {template="Multiply::Execute(in A,in B,out Result)"}.Result
    ExecuteContext=@SequenceExecution_3.B
   ) {variable=look_up_and_left_two_hands_multiplied_alpha}
   RigVMVariableNode(
    Value=&RigVMFunction_MathFloatMul(A=RigVMVariableNode() {variable=head_SPR_look_up_alpha}.Value, B=RigVMVariableNode() {variable=upperarm_r_rifle_up_alpha}.Value) {template="Multiply::Execute(in A,in B,out Result)"}.Result
    ExecuteContext=@SequenceExecution_3.C
   ) {variable=look_up_two_hands_rifle_multiplied_alpha}
   RigVMVariableNode(
    Value=&RigVMFunction_MathFloatMul(B=RigVMVariableNode() {variable=head_SPR_look_down_alpha}.Value, A=RigVMVariableNode() {variable=upperarm_r_SPR_back_alpha}.Value) {template="Multiply::Execute(in A,in B,out Result)"}.Result
    ExecuteContext=@SequenceExecution_3.D
   ) {variable=look_down_upperarm_r_back_multiplied_alpha}
   SequenceExecution_12 = RigVMAggregateNode(ExecuteContext=@SequenceExecution_12_1.B)
   &RigUnit_ModifyTransforms(
    ItemToModify="((Item=(Type=Bone,Name=\"dyn_hair_short_bk_r_mstr\"),Transform=(Rotation=(X=0,Y=0,Z=0,W=1),Translation=(X=-2,Y=0,Z=-1),Scale3D=(X=1,Y=1,Z=1))),(Item=(Type=Bone,Name=\"dyn_hair_short_side_r_mstr\"),Transform=(Rotation=(X=0,Y=0,Z=0,W=1),Translation=(X=0,Y=0,Z=0) ...(truncated)"
    ItemToModify.0.Transform.Rotation=&RigVMFunction_MathQuaternionFromEuler(Euler=(X=0,Y=0,Z=0)).Result
    ItemToModify.1.Transform.Rotation=&RigVMFunction_MathQuaternionFromEuler(Euler=(X=0,Y=0,Z=0)).Result
    ExecutePin=@SequenceExecution_12.A
    Weight=RigVMVariableNode() {variable=head_SPR_look_up_alpha}.Value
   )
   > &RigUnit_ModifyTransforms(
    ItemToModify="((Item=(Type=Bone,Name=\"dyn_hair_short_fr_r_01\"),Transform=(Rotation=(X=0,Y=0,Z=0,W=1),Translation=(X=0,Y=0,Z=0),Scale3D=(X=1,Y=1,Z=1))),(Item=(Type=Bone,Name=\"dyn_hair_short_fr_l_01\"),Transform=(Rotation=(X=0,Y=0,Z=0,W=1),Translation=(X=0,Y=0,Z=0),Scale3D ...(truncated)"
    ItemToModify.1.Transform.Rotation=&RigVMFunction_MathQuaternionFromEuler(Euler=(X=0,Y=-5,Z=15)).Result
    ItemToModify.0.Transform.Rotation=&RigVMFunction_MathQuaternionFromEuler(Euler=(X=0,Y=-5,Z=-15)).Result
    Weight=RigVMVariableNode() {variable=head_SPR_look_down_alpha}.Value
   )
   n9 = RigVMVariableNode() {variable=look_up_arm_back}
   > &RigUnit_ModifyTransforms(
    ItemToModify=((Item=(Type=Bone,Name="dyn_crown_short_mstr"),Transform=(Rotation=(X=0,Y=0,Z=0,W=1),Translation=(X=-1,Y=0,Z=0),Scale3D=(X=1,Y=1,Z=1))))
    ItemToModify.0.Transform.Rotation=&RigVMFunction_MathQuaternionFromEuler(Euler=(X=0,Y=-5,Z=-5)).Result
    Weight=@n9.Value
   )
   > &RigUnit_ModifyTransforms(
    ItemToModify="((Item=(Type=Bone,Name=\"dyn_hair_short_bk_r_mstr\"),Transform=(Rotation=(X=0,Y=0,Z=0,W=1),Translation=(X=-1,Y=0,Z=-1),Scale3D=(X=1,Y=1,Z=1))),(Item=(Type=Bone,Name=\"dyn_hair_short_side_r_mstr\"),Transform=(Rotation=(X=0,Y=0,Z=0,W=1),Translation=(X=-2,Y=0,Z=0 ...(truncated)"
    Weight=@n9.Value
   )
   n10 = RigVMVariableNode() {variable=look_up_and_left_two_hands_multiplied_alpha}
   &RigUnit_ModifyTransforms(
    ItemToModify="((Item=(Type=Bone,Name=\"dyn_hair_short_bk_r_mstr\"),Transform=(Rotation=(X=0,Y=0,Z=0,W=1),Translation=(X=-2,Y=0,Z=-1),Scale3D=(X=1,Y=1,Z=1))),(Item=(Type=Bone,Name=\"dyn_hair_short_side_r_mstr\"),Transform=(Rotation=(X=0,Y=0,Z=0,W=1),Translation=(X=-1,Y=0,Z=0 ...(truncated)"
    ItemToModify.2.Transform.Rotation=&RigVMFunction_MathQuaternionFromEuler(Euler=(X=0,Y=15,Z=0)).Result
    ItemToModify.1.Transform.Rotation=&RigVMFunction_MathQuaternionFromEuler(Euler=(X=0,Y=15,Z=0)).Result
    ItemToModify.0.Transform.Rotation=&RigVMFunction_MathQuaternionFromEuler(Euler=(X=0,Y=-10,Z=0)).Result
    Weight=@n10.Value
    ExecutePin=@SequenceExecution_12.B
   )
   > &RigUnit_ModifyTransforms(
    ItemToModify="((Item=(Type=Bone,Name=\"dyn_hair_short_fr_r_02\"),Transform=(Rotation=(X=0,Y=0,Z=0,W=1),Translation=(X=-0.5,Y=0,Z=0),Scale3D=(X=1,Y=1,Z=1))),(Item=(Type=Bone,Name=\"dyn_hair_short_side_r_02\"),Transform=(Rotation=(X=0,Y=0,Z=0,W=1),Translation=(X=-3,Y=0,Z=0),S ...(truncated)"
    ItemToModify.0.Transform.Rotation=&RigVMFunction_MathQuaternionFromEuler(Euler=(X=-20,Y=10,Z=0)).Result
    ItemToModify.2.Transform.Rotation=&RigVMFunction_MathQuaternionFromEuler(Euler=(X=0,Y=-5,Z=0)).Result
    Weight=@n10.Value
   )
   > &RigUnit_ModifyTransforms(
    ItemToModify="((Item=(Type=Bone,Name=\"dyn_hair_short_side_l_mstr\"),Transform=(Rotation=(X=0,Y=0,Z=0,W=1),Translation=(X=0,Y=0,Z=0),Scale3D=(X=1,Y=1,Z=1))),(Item=(Type=Bone,Name=\"dyn_hair_short_bk_l_mstr\"),Transform=(Rotation=(X=0,Y=0,Z=0,W=1),Translation=(X=-1,Y=0,Z=0), ...(truncated)"
    ItemToModify.0.Transform.Rotation=&RigVMFunction_MathQuaternionFromEuler(Euler=(X=0,Y=15,Z=0)).Result
    ItemToModify.1.Transform.Rotation=&RigVMFunction_MathQuaternionFromEuler(Euler=(X=-15,Y=15,Z=0)).Result
    Weight=@n10.Value
   )
   > &RigUnit_ModifyTransforms(
    ItemToModify="((Item=(Type=Bone,Name=\"dyn_hair_short_bk_r_mstr\"),Transform=(Rotation=(X=0,Y=0,Z=0,W=1),Translation=(X=-1,Y=0,Z=-1),Scale3D=(X=1,Y=1,Z=1))),(Item=(Type=Bone,Name=\"dyn_hair_short_side_r_mstr\"),Transform=(Rotation=(X=0,Y=0,Z=0,W=1),Translation=(X=-3,Y=0,Z=0 ...(truncated)"
    ItemToModify.0.Transform.Rotation=&RigVMFunction_MathQuaternionFromEuler(Euler=(X=0,Y=0,Z=0)).Result
    Weight=RigVMVariableNode() {variable=look_up_two_hands_rifle_multiplied_alpha}.Value
   )
   SequenceExecution_12_1_1 = RigVMAggregateNode(ExecuteContext=@RigVMFunction_ControlFlowBranch.True)
   SequenceExecution_3_1 = RigVMAggregateNode(ExecuteContext=@SequenceExecution_12_1_1.A)
   SequenceExecution_4_1 = RigVMAggregateNode(ExecuteContext=@SequenceExecution_3_1.A)
   SequenceExecution_7_1 = RigVMAggregateNode(ExecuteContext=@SequenceExecution_4_1.A)
   n11 = &RigUnit_SphericalPoseReader(
    ActiveRegionScaleFactors=(PositiveWidth=1,NegativeWidth=1,PositiveHeight=1,NegativeHeight=1)
    ActiveRegionSize=0.065
    Debug=(bDrawDebug=false,DebugScale=25,DebugSegments=20,DebugThickness=0.25)
    DriverAxis=(X=-1,Y=0,Z=0)
    DriverItem=(Type=Bone,Name="upperarm_r")
    FalloffRegionScaleFactors=(PositiveWidth=1,NegativeWidth=1,PositiveHeight=1,NegativeHeight=1)
    FalloffSize=0.125
    OptionalParentItem=(Type=Bone)
    RotationOffset=(X=55,Y=10,Z=80)
    ExecutePin=@SequenceExecution_7_1.A
   )
   > RigVMVariableNode(Value=@n11.OutputParam) {variable=upperarm_r_SPR_back_alpha}
   n12 = &RigUnit_SphericalPoseReader(
    ActiveRegionScaleFactors=(PositiveWidth=1,NegativeWidth=1,PositiveHeight=1,NegativeHeight=1)
    ActiveRegionSize=0.075
    Debug=(bDrawDebug=true,DebugScale=25,DebugSegments=20,DebugThickness=0.25)
    DriverAxis=(X=-1,Y=0,Z=0)
    DriverItem=(Type=Bone,Name="upperarm_r")
    FalloffRegionScaleFactors=(PositiveWidth=1,NegativeWidth=1,PositiveHeight=1,NegativeHeight=1)
    FalloffSize=0.2
    OptionalParentItem=(Type=Bone)
    RotationOffset=(X=-55,Y=0,Z=-10)
    ExecutePin=@SequenceExecution_7_1.D
   )
   > RigVMVariableNode(Value=@n12.OutputParam) {variable=upperarm_r_rifle_up_alpha}
   n13 = &RigUnit_SphericalPoseReader(
    ActiveRegionScaleFactors=(PositiveWidth=1.88,NegativeWidth=1.8,PositiveHeight=2,NegativeHeight=2.5)
    ActiveRegionSize=0.065
    Debug=(bDrawDebug=true,DebugScale=25,DebugSegments=20,DebugThickness=0.25)
    DriverAxis=(X=-1,Y=0,Z=0)
    DriverItem=(Type=Bone,Name="upperarm_r")
    FalloffRegionScaleFactors=(PositiveWidth=1,NegativeWidth=1.8,PositiveHeight=1.8,NegativeHeight=1)
    FalloffSize=0.25
    OptionalParentItem=(Type=Bone)
    RotationOffset=(X=-115,Y=15,Z=25)
    ExecutePin=@SequenceExecution_7_1.B
   )
   > RigVMVariableNode(Value=@n13.OutputParam) {variable=uppearm_r_SPR_aim_up_two_hands_alpha}
   n14 = &RigUnit_SphericalPoseReader(
    ActiveRegionScaleFactors=(PositiveWidth=2,NegativeWidth=2,PositiveHeight=1,NegativeHeight=1)
    ActiveRegionSize=0.05
    Debug=(bDrawDebug=false,DebugScale=25,DebugSegments=20,DebugThickness=0.25)
    DriverAxis=(X=-1,Y=0,Z=0)
    DriverItem=(Type=Bone,Name="upperarm_r")
    FalloffRegionScaleFactors=(PositiveWidth=1,NegativeWidth=1,PositiveHeight=0.75,NegativeHeight=0.75)
    FalloffSize=0.2
    FlipWidthScaling=false
    OptionalParentItem=(Type=Bone)
    RotationOffset=(X=15,Y=150,Z=0)
    ExecutePin=@SequenceExecution_7_1.C
   )
   > RigVMVariableNode(Value=@n14.OutputParam) {variable=upperarm_r_skydiving}
   SequenceExecution_2_1 = RigVMAggregateNode(ExecuteContext=@SequenceExecution_4_1.B)
   n15 = &RigUnit_SphericalPoseReader(
    ActiveRegionScaleFactors=(PositiveWidth=1,NegativeWidth=0.7,PositiveHeight=1.5,NegativeHeight=1.5)
    ActiveRegionSize=0.1
    Debug=(bDrawDebug=true,DebugScale=25,DebugSegments=20,DebugThickness=0.25)
    DriverAxis=(X=0,Y=1,Z=0)
    DriverItem=(Type=Bone,Name="head")
    FalloffRegionScaleFactors=(PositiveWidth=1,NegativeWidth=0.7,PositiveHeight=1.5,NegativeHeight=1.5)
    FalloffSize=0.2
    FlipWidthScaling=false
    OptionalParentItem=(Type=Bone)
    RotationOffset=(X=90,Y=0,Z=-33)
    ExecutePin=@SequenceExecution_2_1.A
   )
   > RigVMVariableNode(Value=@n15.OutputParam) {variable=head_SPR_look_up_alpha}
   n16 = &RigUnit_SphericalPoseReader(
    ActiveRegionScaleFactors=(PositiveWidth=2,NegativeWidth=0.5,PositiveHeight=2,NegativeHeight=2)
    ActiveRegionSize=0.075
    Debug=(bDrawDebug=false,DebugScale=25,DebugSegments=20,DebugThickness=0.25)
    DriverAxis=(X=0,Y=1,Z=0)
    DriverItem=(Type=Bone,Name="head")
    FalloffRegionScaleFactors=(PositiveWidth=1,NegativeWidth=0.4,PositiveHeight=1,NegativeHeight=1)
    FalloffSize=0.2
    FlipHeightScaling=false
    FlipWidthScaling=false
    OptionalParentItem=(Type=Bone)
    RotationOffset=(X=90,Y=0,Z=12.5)
    ExecutePin=@SequenceExecution_2_1.B
   )
   > RigVMVariableNode(Value=@n16.OutputParam) {variable=head_SPR_look_down_alpha}
   n17 = &RigUnit_SphericalPoseReader(
    ActiveRegionScaleFactors=(PositiveWidth=0.75,NegativeWidth=0.75,PositiveHeight=1,NegativeHeight=0.5)
    ActiveRegionSize=0.1
    Debug=(bDrawDebug=false,DebugScale=25,DebugSegments=20,DebugThickness=0.25)
    DriverAxis=(X=0,Y=1,Z=0)
    DriverItem=(Type=Bone,Name="head")
    FalloffRegionScaleFactors=(PositiveWidth=1,NegativeWidth=1,PositiveHeight=1,NegativeHeight=0.5)
    FalloffSize=0.2
    OptionalParentItem=(Type=Bone)
    RotationOffset=(X=120,Y=0,Z=0)
    ExecutePin=@SequenceExecution_2_1.C
   )
   > RigVMVariableNode(Value=@n17.OutputParam) {variable=head_SPR_look_left_alpha}
   n18 = &RigUnit_SphericalPoseReader(
    ActiveRegionScaleFactors=(PositiveWidth=0.75,NegativeWidth=0.75,PositiveHeight=1,NegativeHeight=0.5)
    ActiveRegionSize=0.1
    Debug=(bDrawDebug=false,DebugScale=25,DebugSegments=20,DebugThickness=0.25)
    DriverAxis=(X=0,Y=1,Z=0)
    DriverItem=(Type=Bone,Name="head")
    FalloffRegionScaleFactors=(PositiveWidth=1,NegativeWidth=1,PositiveHeight=1,NegativeHeight=0.5)
    FalloffSize=0.2
    FlipHeightScaling=false
    FlipWidthScaling=false
    OptionalParentItem=(Type=Bone)
    RotationOffset=(X=60,Y=0,Z=0)
    ExecutePin=@SequenceExecution_2_1.D
   )
   > RigVMVariableNode(Value=@n18.OutputParam) {variable=head_SPR_look_right_alpha}
   RigVMVariableNode(
    Value=&RigVMFunction_MathFloatMul(B=RigVMVariableNode() {variable=upperarm_r_SPR_back_alpha}.Value, A=RigVMVariableNode() {variable=head_SPR_look_up_alpha}.Value) {template="Multiply::Execute(in A,in B,out Result)"}.Result
    ExecuteContext=@SequenceExecution_3_1.E
   ) {variable=look_up_arm_back}
   RigVMVariableNode(
    Value=&RigVMFunction_MathFloatMul(B=RigVMVariableNode() {variable=upperarm_r_rifle_up_alpha}.Value, A=RigVMVariableNode() {variable=crouch_look_up_alpha}.Value) {template="Multiply::Execute(in A,in B,out Result)"}.Result
    ExecuteContext=@SequenceExecution_3_1.F
   ) {variable=crouch_look_rifle}
   RigVMVariableNode(
    Value=&RigVMFunction_MathFloatMul(B=RigVMVariableNode() {variable=uppearm_r_SPR_aim_up_two_hands_alpha}.Value, A=RigVMVariableNode() {variable=crouch_look_up_alpha}.Value) {template="Multiply::Execute(in A,in B,out Result)"}.Result
    ExecuteContext=@SequenceExecution_3_1.G
   ) {variable=crouch_look_pistol}
   RigVMVariableNode(
    Value=&RigVMFunction_MathFloatMul(B=RigVMVariableNode() {variable=uppearm_r_SPR_aim_up_two_hands_alpha}.Value, A=RigVMVariableNode() {variable=head_SPR_look_up_alpha}.Value) {template="Multiply::Execute(in A,in B,out Result)"}.Result
    ExecuteContext=@SequenceExecution_3_1.B
   ) {variable=look_up_and_left_two_hands_multiplied_alpha}
   RigVMVariableNode(
    Value=&RigVMFunction_MathFloatMul(A=RigVMVariableNode() {variable=head_SPR_look_up_alpha}.Value, B=RigVMVariableNode() {variable=upperarm_r_rifle_up_alpha}.Value) {template="Multiply::Execute(in A,in B,out Result)"}.Result
    ExecuteContext=@SequenceExecution_3_1.C
   ) {variable=look_up_two_hands_rifle_multiplied_alpha}
   RigVMVariableNode(
    Value=&RigVMFunction_MathFloatMul(B=RigVMVariableNode() {variable=head_SPR_look_down_alpha}.Value, A=RigVMVariableNode() {variable=upperarm_r_SPR_back_alpha}.Value) {template="Multiply::Execute(in A,in B,out Result)"}.Result
    ExecuteContext=@SequenceExecution_3_1.D
   ) {variable=look_down_upperarm_r_back_multiplied_alpha}
   n19 = &RigUnit_SphericalPoseReader(
    ActiveRegionScaleFactors=(PositiveWidth=2,NegativeWidth=0.5,PositiveHeight=2,NegativeHeight=2)
    ActiveRegionSize=0.075
    Debug=(bDrawDebug=true,DebugScale=25,DebugSegments=20,DebugThickness=0.25)
    DriverAxis=(X=1,Y=0,Z=0)
    DriverItem=(Type=Bone,Name="head")
    FalloffRegionScaleFactors=(PositiveWidth=1,NegativeWidth=0.4,PositiveHeight=1,NegativeHeight=1)
    FalloffSize=0.25
    FlipHeightScaling=false
    FlipWidthScaling=true
    OptionalParentItem=(Type=Bone,Name="spine_05")
    RotationOffset=(X=-70,Y=5,Z=0)
    ExecutePin=@SequenceExecution_12_1_1.B
   )
   > RigVMVariableNode(Value=@n19.OutputParam) {variable=crouch_look_up_alpha}
   n20 = RigVMVariableNode() {variable=crouch_look_pistol}
   &RigUnit_ModifyTransforms(
    ItemToModify="((Item=(Type=Bone,Name=\"dyn_hair_short_bk_r_mstr\"),Transform=(Rotation=(X=0,Y=0,Z=0,W=1),Translation=(X=-5,Y=2,Z=4),Scale3D=(X=1,Y=1,Z=1))),(Item=(Type=Bone,Name=\"dyn_hair_short_side_r_mstr\"),Transform=(Rotation=(X=0,Y=0,Z=0,W=1),Translation=(X=-4,Y=0,Z=0) ...(truncated)"
    ItemToModify.2.Transform.Rotation=&RigVMFunction_MathQuaternionFromEuler(Euler=(X=10,Y=60,Z=0)).Result
    ItemToModify.1.Transform.Rotation=&RigVMFunction_MathQuaternionFromEuler(Euler=(X=0,Y=20,Z=0)).Result
    ItemToModify.0.Transform.Rotation=&RigVMFunction_MathQuaternionFromEuler(Euler=(X=0,Y=-20,Z=45)).Result
    ExecutePin=@SequenceExecution_12_1_1.C
    Weight=@n20.Value
   )
   > &RigUnit_ModifyTransforms(
    ItemToModify=((Item=(Type=Bone,Name="dyn_crown_short_mstr"),Transform=(Rotation=(X=0,Y=0,Z=0,W=1),Translation=(X=2,Y=-2,Z=0),Scale3D=(X=1,Y=1,Z=1))))
    ItemToModify.0.Transform.Rotation=&RigVMFunction_MathQuaternionFromEuler(Euler=(X=0,Y=-10,Z=0)).Result
    Weight=@n20.Value
   )
   n21 = RigVMVariableNode() {variable=crouch_look_rifle}
   > &RigUnit_ModifyTransforms(
    ItemToModify="((Item=(Type=Bone,Name=\"dyn_hair_short_bk_r_mstr\"),Transform=(Rotation=(X=0,Y=0,Z=0,W=1),Translation=(X=-10,Y=-3,Z=0),Scale3D=(X=1,Y=1,Z=1))),(Item=(Type=Bone,Name=\"dyn_hair_short_side_r_mstr\"),Transform=(Rotation=(X=0,Y=0,Z=0,W=1),Translation=(X=-4,Y=0,Z= ...(truncated)"
    ItemToModify.2.Transform.Rotation=&RigVMFunction_MathQuaternionFromEuler(Euler=(X=-10,Y=-30,Z=0)).Result
    ItemToModify.1.Transform.Rotation=&RigVMFunction_MathQuaternionFromEuler(Euler=(X=0,Y=0,Z=0)).Result
    ItemToModify.0.Transform.Rotation=&RigVMFunction_MathQuaternionFromEuler(Euler=(X=10,Y=-20,Z=0)).Result
    Weight=@n21.Value
   )
   > &RigUnit_ModifyTransforms(
    ItemToModify=((Item=(Type=Bone,Name="dyn_crown_short_mstr"),Transform=(Rotation=(X=0,Y=0,Z=0,W=1),Translation=(X=-1,Y=0,Z=0),Scale3D=(X=1,Y=1,Z=1))))
    ItemToModify.0.Transform.Rotation=&RigVMFunction_MathQuaternionFromEuler(Euler=(X=0,Y=-5,Z=0)).Result
    Weight=@n21.Value
   )
   n22 = RigVMVariableNode() {variable=crouch_look_up_alpha}
   > &RigUnit_ModifyTransforms(
    ItemToModify="((Item=(Type=Bone,Name=\"dyn_hair_short_bk_r_mstr\"),Transform=(Rotation=(X=0,Y=0,Z=0,W=1),Translation=(X=-1,Y=-1,Z=0),Scale3D=(X=1,Y=1,Z=1))),(Item=(Type=Bone,Name=\"dyn_hair_short_side_r_mstr\"),Transform=(Rotation=(X=0,Y=0,Z=0,W=1),Translation=(X=-4,Y=0,Z=0 ...(truncated)"
    ItemToModify.2.Transform.Rotation=&RigVMFunction_MathQuaternionFromEuler(Euler=(X=-10,Y=-60,Z=0)).Result
    ItemToModify.0.Transform.Rotation=&RigVMFunction_MathQuaternionFromEuler(Euler=(X=10,Y=-20,Z=0)).Result
    ItemToModify.1.Transform.Rotation=&RigVMFunction_MathQuaternionFromEuler(Euler=(X=0,Y=0,Z=0)).Result
    Weight=@n22.Value
   )
   > &RigUnit_ModifyTransforms(
    ItemToModify=((Item=(Type=Bone,Name="dyn_crown_short_mstr"),Transform=(Rotation=(X=0,Y=0,Z=0,W=1),Translation=(X=-1,Y=0,Z=0),Scale3D=(X=1,Y=1,Z=1))))
    Weight=1
    ItemToModify.0.Transform.Rotation=&RigVMFunction_MathQuaternionFromEuler(Euler=(X=0,Y=-15,Z=0)).Result
   )
   > &RigUnit_ModifyTransforms(
    ItemToModify=((Item=(Type=Bone,Name="dyn_crown_short_mstr"),Transform=(Rotation=(X=0,Y=0,Z=0,W=1),Translation=(X=-2,Y=1,Z=0),Scale3D=(X=1,Y=1,Z=1))))
    ItemToModify.0.Transform.Rotation=&RigVMFunction_MathQuaternionFromEuler(Euler=(X=0,Y=-5,Z=0)).Result
    Weight=@n22.Value
   )
   n23 = RigVMVariableNode() {variable=look_up_and_left_two_hands_multiplied_alpha}
   > &RigUnit_ModifyTransforms(
    ItemToModify="((Item=(Type=Bone,Name=\"dyn_hair_short_bk_r_mstr\"),Transform=(Rotation=(X=0,Y=0,Z=0,W=1),Translation=(X=-3,Y=1,Z=0),Scale3D=(X=1,Y=1,Z=1))),(Item=(Type=Bone,Name=\"dyn_hair_short_side_r_mstr\"),Transform=(Rotation=(X=0,Y=0,Z=0,W=1),Translation=(X=0,Y=0,Z=0), ...(truncated)"
    ItemToModify.0.Transform.Rotation=&RigVMFunction_MathQuaternionFromEuler(Euler=(X=20,Y=-40,Z=0)).Result
    ItemToModify.1.Transform.Rotation=&RigVMFunction_MathQuaternionFromEuler(Euler=(X=0,Y=0,Z=0)).Result
    Weight=@n23.Value
   )
   > &RigUnit_ModifyTransforms(
    ItemToModify=((Item=(Type=Bone,Name="dyn_crown_short_mstr"),Transform=(Rotation=(X=0,Y=0,Z=0,W=1),Translation=(X=-1,Y=0,Z=0),Scale3D=(X=1,Y=1,Z=1))))
    ItemToModify.0.Transform.Rotation=&RigVMFunction_MathQuaternionFromEuler(Euler=(X=0,Y=0,Z=0)).Result
    Weight=@n23.Value
   )
   > &RigUnit_ModifyTransforms(
    ItemToModify="((Item=(Type=Bone,Name=\"dyn_hair_short_fr_r_02\"),Transform=(Rotation=(X=0,Y=0,Z=0,W=1),Translation=(X=0,Y=0,Z=0),Scale3D=(X=1,Y=1,Z=1))),(Item=(Type=Bone,Name=\"dyn_hair_short_side_r_02\"),Transform=(Rotation=(X=0,Y=0,Z=0,W=1),Translation=(X=-3,Y=0,Z=0),Scal ...(truncated)"
    ItemToModify.2.Transform.Rotation=&RigVMFunction_MathQuaternionFromEuler(Euler=(X=-40,Y=0,Z=0)).Result
    ItemToModify.0.Transform.Rotation=&RigVMFunction_MathQuaternionFromEuler(Euler=(X=0,Y=0,Z=0)).Result
    Weight=@n23.Value
   )
   > &RigUnit_ModifyTransforms(
    ItemToModify="((Item=(Type=Bone,Name=\"dyn_hair_short_side_l_mstr\"),Transform=(Rotation=(X=0,Y=0,Z=0,W=1),Translation=(X=0,Y=0,Z=0),Scale3D=(X=1,Y=1,Z=1))),(Item=(Type=Bone,Name=\"dyn_hair_short_bk_l_mstr\"),Transform=(Rotation=(X=0,Y=0,Z=0,W=1),Translation=(X=-1,Y=-2,Z=0) ...(truncated)"
    ItemToModify.1.Transform.Rotation=&RigVMFunction_MathQuaternionFromEuler(Euler=(X=0,Y=0,Z=-30)).Result
    ItemToModify.0.Transform.Rotation=&RigVMFunction_MathQuaternionFromEuler(Euler=(X=0,Y=20,Z=0)).Result
    Weight=@n23.Value
   )
   n24 = RigVMVariableNode() {variable=look_up_two_hands_rifle_multiplied_alpha}
   > &RigUnit_ModifyTransforms(
    ItemToModify="((Item=(Type=Bone,Name=\"dyn_hair_short_bk_r_mstr\"),Transform=(Rotation=(X=0,Y=0,Z=0,W=1),Translation=(X=-3,Y=3,Z=0),Scale3D=(X=1,Y=1,Z=1))),(Item=(Type=Bone,Name=\"dyn_hair_short_side_r_mstr\"),Transform=(Rotation=(X=0,Y=0,Z=0,W=1),Translation=(X=0,Y=0,Z=0), ...(truncated)"
    ItemToModify.1.Transform.Rotation=&RigVMFunction_MathQuaternionFromEuler(Euler=(X=-15,Y=0,Z=0)).Result
    ItemToModify.2.Transform.Rotation=&RigVMFunction_MathQuaternionFromEuler(Euler=(X=0,Y=0,Z=0)).Result
    ItemToModify.3.Transform.Rotation=&RigVMFunction_MathQuaternionFromEuler(Euler=(X=0,Y=0,Z=0)).Result
    ItemToModify.0.Transform.Rotation=&RigVMFunction_MathQuaternionFromEuler(Euler=(X=0,Y=40,Z=-30)).Result
    Weight=@n24.Value
   )
   > &RigUnit_ModifyTransforms(
    ItemToModify="((Item=(Type=Bone,Name=\"dyn_hair_short_bk_r_mstr\"),Transform=(Rotation=(X=0,Y=0,Z=0,W=1),Translation=(X=-5,Y=0,Z=1.5),Scale3D=(X=1,Y=1,Z=1))),(Item=(Type=Bone,Name=\"dyn_hair_short_side_r_mstr\"),Transform=(Rotation=(X=0,Y=0,Z=0,W=1),Translation=(X=-6,Y=0,Z= ...(truncated)"
    ItemToModify.1.Transform.Rotation=&RigVMFunction_MathQuaternionFromEuler(Euler=(X=0,Y=0,Z=10)).Result
    ItemToModify.3.Transform.Rotation=&RigVMFunction_MathQuaternionFromEuler(Euler=(X=0,Y=-10,Z=0)).Result
    ItemToModify.0.Transform.Rotation=&RigVMFunction_MathQuaternionFromEuler(Euler=(X=0,Y=-25,Z=0)).Result
    Weight=RigVMVariableNode() {variable=look_up_arm_back}.Value
   )
   > &RigUnit_ModifyTransforms(
    ItemToModify=((Item=(Type=Bone,Name="dyn_crown_short_mstr"),Transform=(Rotation=(X=0,Y=0,Z=0,W=1),Translation=(X=-1,Y=0,Z=0),Scale3D=(X=1,Y=1,Z=1))))
    ItemToModify.0.Transform.Rotation=&RigVMFunction_MathQuaternionFromEuler(Euler=(X=0,Y=-7,Z=-10)).Result
    Weight=@n24.Value
   )
  ]}
 SequenceExecution_12_2_ContainedGraph {nodes [
   n1 = RigVMFunctionEntryNode()
   SequenceExecution_12 = &RigVMFunction_Sequence(ExecuteContext=@n1.ExecuteContext)
   SequenceExecution_12_1 = &RigVMFunction_Sequence(ExecuteContext=@SequenceExecution_12.B)
   RigVMFunctionReturnNode(A=@SequenceExecution_12.A, B=@SequenceExecution_12_1.A, C=@SequenceExecution_12_1.B)
  ]}
]
hierarchy [
 root {loc=<0 0 0>, rot=<0 0 0 1>}
 pelvis {parent=root, loc=<0 -0.5301 98.0571>, rot=<0.0127 0.707 -0.0127 -0.707>}
 spine_01 {parent=pelvis, loc=<6.1208 -0.9328 0>, rot=<0 0 0.034 -0.9994>}
 spine_02 {parent=spine_01, loc=<6.5996 -0.2095 0>, rot=<0 0 0.0238 -0.9997>}
 spine_03 {parent=spine_02, loc=<6.4914 -0.1237 0>, rot=<0 0 0.0149 -0.9999>}
 spine_04 {parent=spine_03, loc=<6.9598 0.0794 0>, rot=<0 0 -0.0132 -0.9999>}
 spine_05 {parent=spine_04, loc=<6.9987 0 0>, rot=<0 0 -0.0015 -1>}
 neck_01 {parent=spine_05, loc=<11.3138 0.1529 0>, rot=<0 0 -0.2003 -0.9797>}
 neck_02 {parent=neck_01, loc=<4.5627 0 0>, rot=<0 0 0.0504 -0.9987>}
 head {parent=neck_02, loc=<5.3914 0 0>, rot=<0 0 0.115 -0.9934>}
 dfrm_hair_short_mstr {parent=head, loc=<9.6892 -1.5079 0>, rot=<0.5021 -0.4979 -0.5021 -0.4979>}
 dyn_hair_short_bangs_r_mstr {parent=dfrm_hair_short_mstr, loc=<-4.837 2.6943 8.9341>, rot=<0.1069 0.1069 -0.699 -0.699>}
 dyn_hair_short_bangs_r_01 {parent=dyn_hair_short_bangs_r_mstr, loc=<0 0 0>, rot=<0 0 0 1>}
 dyn_hair_short_bangs_l_mstr {parent=dfrm_hair_short_mstr, loc=<3.9682 3.7549 9.9609>, rot=<-0.1183 -0.1183 -0.6971 -0.6971>}
 dyn_hair_short_bangs_l_01 {parent=dyn_hair_short_bangs_l_mstr, loc=<0 0 0>, rot=<0 0 0 1>}
 dyn_hair_short_bangs_c_mstr {parent=dfrm_hair_short_mstr, loc=<-0.2554 2.521 10.8217>, rot=<-0.027 -0.027 -0.7066 -0.7066>}
 dyn_hair_short_bangs_c_01 {parent=dyn_hair_short_bangs_c_mstr, loc=<0 0 0>, rot=<0 0 0 1>}
 dyn_hair_short_side_r_mstr {parent=dfrm_hair_short_mstr, loc=<-6.815 8.6093 -2.4006>, rot=<0.5953 0.5953 -0.3816 -0.3816>}
 dyn_hair_short_side_r_01 {parent=dyn_hair_short_side_r_mstr, loc=<0 0 0>, rot=<0 0 0 1>}
 dyn_hair_short_side_r_02 {parent=dyn_hair_short_side_r_01, loc=<6.6085 0 0>, rot=<0 0 0 1>}
 dyn_hair_short_fr_r_mstr {parent=dfrm_hair_short_mstr, loc=<-7.0478 4.0589 4.4053>, rot=<0.5 0.5 -0.5 -0.5>}
 dyn_hair_short_fr_r_01 {parent=dyn_hair_short_fr_r_mstr, loc=<0 0 0>, rot=<0 0 0 1>}
 dyn_hair_short_fr_r_02 {parent=dyn_hair_short_fr_r_01, loc=<9.4825 0 0>, rot=<0 0 0 1>}
 dyn_hair_short_bk_r_mstr {parent=dfrm_hair_short_mstr, loc=<-2.2928 11.6719 -7.418>, rot=<0.7027 0.7027 -0.079 -0.079>}
 dyn_hair_short_bk_r_01 {parent=dyn_hair_short_bk_r_mstr, loc=<0 0 0>, rot=<0 0 0 1>}
 dyn_hair_short_bk_r_02 {parent=dyn_hair_short_bk_r_01, loc=<4.5196 0 0>, rot=<0 0 0 1>}
 dyn_hair_short_bk_l_mstr {parent=dfrm_hair_short_mstr, loc=<5.4946 12.6902 -5.3369>, rot=<0.6565 0.6565 0.2627 0.2627>}
 dyn_hair_short_bk_l_01 {parent=dyn_hair_short_bk_l_mstr, loc=<0 0 0>, rot=<0 0 0 1>}
 dyn_hair_short_bk_l_02 {parent=dyn_hair_short_bk_l_01, loc=<3.5303 0 0>, rot=<0 0 0 1>}
 dyn_hair_short_side_l_mstr {parent=dfrm_hair_short_mstr, loc=<7.7084 9.4883 0.1982>, rot=<-0.5 -0.5 -0.5 -0.5>}
 dyn_hair_short_side_l_01 {parent=dyn_hair_short_side_l_mstr, loc=<0 0 0>, rot=<0 0 0 1>}
 dyn_hair_short_side_l_02 {parent=dyn_hair_short_side_l_01, loc=<6.6435 0 0>, rot=<0 0 0 1>}
 dyn_hair_short_fr_l_mstr {parent=dfrm_hair_short_mstr, loc=<7.0478 7.2378 4.4053>, rot=<-0.5 -0.5 -0.5 -0.5>}
 dyn_hair_short_fr_l_01 {parent=dyn_hair_short_fr_l_mstr, loc=<0 0 0>, rot=<0 0 0 1>}
 dyn_hair_short_fr_l_02 {parent=dyn_hair_short_fr_l_01, loc=<7.4064 0 0>, rot=<0 0 0 1>}
 dyn_crown_short_mstr {parent=head, loc=<2.5881 -0.7095 0>, rot=<0.0029 -0.7071 0.7071 -0.0029>}
 dyn_crown_short {parent=dyn_crown_short_mstr, loc=<0 0 0>, rot=<0 0 0 1>}
 clavicle_l {parent=spine_05, loc=<5.4831 -0.1053 -3.2946>, rot=<0.0465 -0.7289 -0.009 -0.6829>}
 upperarm_l {parent=clavicle_l, loc=<9.8658 -1.2136 -0.4105>, rot=<0.0164 -0.3245 -0.0269 -0.9454>}
 clavicle_r {parent=spine_05, loc=<5.4831 -0.1053 3.2946>, rot=<0.7289 0.0465 -0.6829 0.009>}
 upperarm_r {parent=clavicle_r, loc=<-9.8658 1.2136 0.4104>, rot=<0.0164 -0.3245 -0.0269 -0.9454>}
]
variables [
 Crouch {type=bool, default=False}
 clavicle_r_SPR_back_alpha {type=double, default=0}
 crouch_look_pistol {type=double, default=0}
 crouch_look_rifle {type=double, default=0}
 crouch_look_up_alpha {type=double, default=0}
 head_SPR_look_down_alpha {type=double, default=0}
 head_SPR_look_left_alpha {type=double, default=0}
 head_SPR_look_right_alpha {type=double, default=0}
 head_SPR_look_up_alpha {type=double, default=0}
 look_down_upperarm_r_back_multiplied_alpha {type=double, default=0}
 look_up_and_left_two_hands_multiplied_alpha {type=double, default=0}
 look_up_arm_back {type=double, default=0}
 look_up_two_hands_rifle_multiplied_alpha {type=double, default=0}
 uppearm_r_SPR_aim_up_two_hands_alpha {type=double, default=0}
 upperarm_r_SPR_back_alpha {type=double, default=0}
 upperarm_r_rifle_up_alpha {type=double, default=0}
 upperarm_r_skydiving {type=double, default=0}
]
