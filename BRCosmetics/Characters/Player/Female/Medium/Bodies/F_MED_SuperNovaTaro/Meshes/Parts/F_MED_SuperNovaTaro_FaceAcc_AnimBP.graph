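# Anim Blueprint for the SuperNovaTaro female face-accessory mesh, adding secondary motion to short hair.
# It layers two copied mesh poses at the upper arms, constrains hair master bones to the head, then simulates strands with per-bone AnimDynamics.
# An interpolated Skydiving alpha tilts crown, right-back and right-side bones, before a Control Rig receives the crouch state.
# The EventGraph also sets GravityOverride, Jumping and TargetOrMelee, which the AnimGraph never reads.
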AnimBlueprint /BRCosmetics/Characters/Player/Female/Medium/Bodies/F_MED_SuperNovaTaro/Meshes/Parts/F_MED_SuperNovaTaro_FaceAcc_AnimBP
target_skeleton=/Game/Characters/Player/Female/Medium/Bodies/F_MED_SuperNovaTaro/Meshes/Parts/F_MED_SuperNovaTaro_FaceAcc_Skeleton
graphs [
 AnimGraph {nodes [
   n1 = Copy Pose From Mesh(SourceMeshComponent=Get Skel_Mesh()) {bUseAttachedParent=true}
   n2 = Copy Pose From Mesh() {bUseAttachedParent=true}
   n3 = "Constraint - dyn_hair_short_bangs_l_mstr (P)"(
    ComponentPose=Local To Component(LocalPose=Layered blend per bone(BasePose=@n2, BlendPoses_0=@n1) {layerSetup [{branchFilters [{boneName=upperarm_l}, {boneName=upperarm_r}]}]})
    Alpha=1
    ConstraintWeights_0=1
   ) {
    alphaBoolBlend {blendOption=Linear}
    boneToModify {boneName=dyn_hair_short_bangs_l_mstr}
    constraintSetup [{targetBone {boneName=head}}]
    constraintWeights [1]
   }
   n4 = "Constraint - dyn_hair_short_bangs_c_mstr (P)"(ComponentPose=@n3, Alpha=1, ConstraintWeights_0=1) {
    alphaBoolBlend {blendOption=Linear}
    boneToModify {boneName=dyn_hair_short_bangs_c_mstr}
    constraintSetup [{targetBone {boneName=head}}]
    constraintWeights [1]
   }
   n5 = "Constraint - dyn_hair_short_bangs_r_mstr (P)"(ComponentPose=@n4, Alpha=1, ConstraintWeights_0=1) {
    alphaBoolBlend {blendOption=Linear}
    boneToModify {boneName=dyn_hair_short_bangs_r_mstr}
    constraintSetup [{targetBone {boneName=head}}]
    constraintWeights [1]
   }
   n6 = "Constraint - dyn_hair_short_side_r_mstr (P)"(ComponentPose=@n5, Alpha=1, ConstraintWeights_0=1) {
    alphaBoolBlend {blendOption=Linear}
    boneToModify {boneName=dyn_hair_short_side_r_mstr}
    constraintSetup [{targetBone {boneName=head}}]
    constraintWeights [1]
   }
   n7 = "Constraint - dyn_hair_short_fr_r_mstr (P)"(ComponentPose=@n6, Alpha=1, ConstraintWeights_0=1) {
    alphaBoolBlend {blendOption=Linear}
    boneToModify {boneName=dyn_hair_short_fr_r_mstr}
    constraintSetup [{targetBone {boneName=head}}]
    constraintWeights [1]
   }
   n8 = "Constraint - dyn_hair_short_bk_r_mstr (P)"(ComponentPose=@n7, Alpha=1, ConstraintWeights_0=1) {
    alphaBoolBlend {blendOption=Linear}
    boneToModify {boneName=dyn_hair_short_bk_r_mstr}
    constraintSetup [{targetBone {boneName=head}}]
    constraintWeights [1]
   }
   n9 = "Constraint - dyn_hair_short_bk_l_mstr (P)"(ComponentPose=@n8, Alpha=1, ConstraintWeights_0=1) {
    alphaBoolBlend {blendOption=Linear}
    boneToModify {boneName=dyn_hair_short_bk_l_mstr}
    constraintSetup [{targetBone {boneName=head}}]
    constraintWeights [1]
   }
   n10 = "Constraint - dyn_hair_short_side_l_mstr (P)"(ComponentPose=@n9, Alpha=1, ConstraintWeights_0=1) {
    alphaBoolBlend {blendOption=Linear}
    boneToModify {boneName=dyn_hair_short_side_l_mstr}
    constraintSetup [{targetBone {boneName=head}}]
    constraintWeights [1]
   }
   n11 = "Constraint - dyn_hair_short_fr_l_mstr (P)"(ComponentPose=@n10, Alpha=1, ConstraintWeights_0=1) {
    alphaBoolBlend {blendOption=Linear}
    boneToModify {boneName=dyn_hair_short_fr_l_mstr}
    constraintSetup [{targetBone {boneName=head}}]
    constraintWeights [1]
   }
   n12 = Anim Dynamics - Bone: dyn_hair_short_bangs_l_01(ComponentPose=@n11, Alpha=0.7, ExternalForce="0,0,0") {
    angularDampingOverride=0.9999
    angularSpringConstant=50
    bGravityOverrideInSimSpace=true
    bOverrideAngularDamping=true
    simulationSpace=BoneRelative
    alphaBoolBlend {blendOption=Linear}
    boundBone {boneName=dyn_hair_short_bangs_l_01}
    componentAppliedLinearAccClamp=<2500 2500 2500>
    componentLinearAccScale=<0.7 0.7 0.8>
    componentLinearVelScale=<0.7 0.7 0.8>
    physicsBodyDefinitions [{
      boundBone {boneName=dyn_hair_short_bangs_l_01}
      boxExtents=<10 10 10>
      constraintSetup {angularLimitsMax=<1 10 10>, angularLimitsMin=<-1 -30 -10>, angularTarget=<1 0 0>}
      localJointOffset=<1 0 0>
     }]
    relativeSpaceBone {boneName=root}
   }
   n13 = Anim Dynamics - Bone: dyn_hair_short_bangs_c_01(ComponentPose=@n12, Alpha=0.8, ExternalForce="0,0,0") {
    angularDampingOverride=0.9999
    angularSpringConstant=50
    bGravityOverrideInSimSpace=true
    bOverrideAngularDamping=true
    simulationSpace=BoneRelative
    alphaBoolBlend {blendOption=Linear}
    boundBone {boneName=dyn_hair_short_bangs_c_01}
    componentAppliedLinearAccClamp=<2500 2500 2500>
    componentLinearAccScale=<0.7 0.7 0.8>
    componentLinearVelScale=<0.7 0.7 0.8>
    physicsBodyDefinitions [{
      boundBone {boneName=dyn_hair_short_bangs_c_01}
      boxExtents=<10 10 10>
      constraintSetup {angularLimitsMax=<1 10 10>, angularLimitsMin=<-1 -50 -10>, angularTarget=<1 0 0>}
      localJointOffset=<1 0 0>
     }]
    relativeSpaceBone {boneName=root}
   }
   n14 = Anim Dynamics - Bone: dyn_hair_short_bangs_r_01(ComponentPose=@n13, Alpha=1, ExternalForce="0,0,0") {
    angularDampingOverride=0.9999
    angularSpringConstant=50
    bGravityOverrideInSimSpace=true
    bOverrideAngularDamping=true
    simulationSpace=BoneRelative
    alphaBoolBlend {blendOption=Linear}
    boundBone {boneName=dyn_hair_short_bangs_r_01}
    componentAppliedLinearAccClamp=<2500 2500 2500>
    componentLinearAccScale=<0.7 0.7 0.8>
    componentLinearVelScale=<0.7 0.7 0.8>
    physicsBodyDefinitions [{
      boundBone {boneName=dyn_hair_short_bangs_r_01}
      boxExtents=<10 10 10>
      constraintSetup {angularLimitsMax=<1 10 10>, angularLimitsMin=<-1 -65 -10>, angularTarget=<1 0 0>}
      localJointOffset=<1 0 0>
     }]
    relativeSpaceBone {boneName=root}
   }
   n15 = Anim Dynamics - Bone: dyn_hair_short_fr_l_01(ComponentPose=@n14, Alpha=1, ExternalForce="0,0,0") {
    angularDampingOverride=0.9999
    angularSpringConstant=50
    bGravityOverrideInSimSpace=true
    bOverrideAngularDamping=true
    simulationSpace=BoneRelative
    alphaBoolBlend {blendOption=Linear}
    boundBone {boneName=dyn_hair_short_fr_l_01}
    componentAppliedLinearAccClamp=<2500 2500 2500>
    componentLinearAccScale=<0.7 0.7 0.8>
    componentLinearVelScale=<0.7 0.7 0.8>
    physicsBodyDefinitions [{
      boundBone {boneName=dyn_hair_short_fr_l_01}
      boxExtents=<10 10 10>
      constraintSetup {angularLimitsMax=<1 10 10>, angularLimitsMin=<-1 -10 -10>, angularTarget=<1 0 0>}
      localJointOffset=<1 0 0>
     }]
    relativeSpaceBone {boneName=root}
   }
   n16 = Anim Dynamics - Bone: dyn_hair_short_fr_l_02(ComponentPose=@n15, Alpha=1, ExternalForce="0,0,0") {
    angularDampingOverride=0.9999
    angularSpringConstant=50
    bGravityOverrideInSimSpace=true
    bOverrideAngularDamping=true
    simulationSpace=BoneRelative
    alphaBoolBlend {blendOption=Linear}
    boundBone {boneName=dyn_hair_short_fr_l_02}
    componentAppliedLinearAccClamp=<2500 2500 2500>
    componentLinearAccScale=<0.7 0.7 0.8>
    componentLinearVelScale=<0.7 0.7 0.8>
    physicsBodyDefinitions [{
      boundBone {boneName=dyn_hair_short_fr_l_02}
      boxExtents=<10 10 10>
      constraintSetup {angularLimitsMax=<1 10 10>, angularLimitsMin=<-1 -10 -10>, angularTarget=<1 0 0>}
      localJointOffset=<1 0 0>
     }]
    relativeSpaceBone {boneName=root}
   }
   n17 = Anim Dynamics - Bone: dyn_hair_short_fr_r_01(ComponentPose=@n16, Alpha=1, ExternalForce="0,0,0") {
    angularDampingOverride=0.9999
    angularSpringConstant=50
    bGravityOverrideInSimSpace=true
    bOverrideAngularDamping=true
    simulationSpace=BoneRelative
    alphaBoolBlend {blendOption=Linear}
    boundBone {boneName=dyn_hair_short_fr_r_01}
    componentAppliedLinearAccClamp=<2500 2500 2500>
    componentLinearAccScale=<0.7 0.7 0.8>
    componentLinearVelScale=<0.7 0.7 0.8>
    physicsBodyDefinitions [{
      boundBone {boneName=dyn_hair_short_fr_r_01}
      boxExtents=<10 10 10>
      constraintSetup {angularLimitsMax=<1 10 10>, angularLimitsMin=<-1 -10 -10>, angularTarget=<1 0 0>}
      localJointOffset=<1 0 0>
     }]
    relativeSpaceBone {boneName=root}
   }
   n18 = Anim Dynamics - Bone: dyn_hair_short_fr_r_02(ComponentPose=@n17, Alpha=1, ExternalForce="0,0,0") {
    angularDampingOverride=0.9999
    angularSpringConstant=50
    bGravityOverrideInSimSpace=true
    bOverrideAngularDamping=true
    simulationSpace=BoneRelative
    alphaBoolBlend {blendOption=Linear}
    boundBone {boneName=dyn_hair_short_fr_r_02}
    componentAppliedLinearAccClamp=<2500 2500 2500>
    componentLinearAccScale=<0.7 0.7 0.8>
    componentLinearVelScale=<0.7 0.7 0.8>
    physicsBodyDefinitions [{
      boundBone {boneName=dyn_hair_short_fr_r_02}
      boxExtents=<10 10 10>
      constraintSetup {angularLimitsMax=<1 10 10>, angularLimitsMin=<-1 -10 -10>, angularTarget=<1 0 0>}
      localJointOffset=<1 0 0>
     }]
    relativeSpaceBone {boneName=root}
   }
   n19 = Anim Dynamics - Bone: dyn_hair_short_bk_r_01(ComponentPose=@n18, Alpha=1, ExternalForce="0,0,0") {
    angularDampingOverride=0.9999
    angularSpringConstant=50
    bGravityOverrideInSimSpace=true
    bOverrideAngularDamping=true
    simulationSpace=BoneRelative
    alphaBoolBlend {blendOption=Linear}
    boundBone {boneName=dyn_hair_short_bk_r_01}
    componentAppliedLinearAccClamp=<2500 2500 2500>
    componentLinearAccScale=<0.7 0.7 0.8>
    componentLinearVelScale=<0.7 0.7 0.8>
    physicsBodyDefinitions [{
      boundBone {boneName=dyn_hair_short_bk_r_01}
      boxExtents=<10 10 10>
      constraintSetup {angularLimitsMax=<2 30 15>, angularLimitsMin=<-2 -15 -15>, angularTarget=<1 0 0>}
      localJointOffset=<1 0 0>
     }]
    relativeSpaceBone {boneName=root}
   }
   n20 = Anim Dynamics - Bone: dyn_hair_short_bk_r_02(ComponentPose=@n19, Alpha=1, ExternalForce="0,0,0") {
    angularDampingOverride=0.9999
    angularSpringConstant=50
    bGravityOverrideInSimSpace=true
    bOverrideAngularDamping=true
    simulationSpace=BoneRelative
    alphaBoolBlend {blendOption=Linear}
    boundBone {boneName=dyn_hair_short_bk_r_02}
    componentAppliedLinearAccClamp=<2500 2500 2500>
    componentLinearAccScale=<0.7 0.7 0.8>
    componentLinearVelScale=<0.7 0.7 0.8>
    physicsBodyDefinitions [{
      boundBone {boneName=dyn_hair_short_bk_r_02}
      boxExtents=<10 10 10>
      constraintSetup {angularLimitsMax=<2 30 15>, angularLimitsMin=<-2 -15 -15>, angularTarget=<1 0 0>}
      localJointOffset=<1 0 0>
     }]
    relativeSpaceBone {boneName=root}
   }
   n21 = Anim Dynamics - Bone: dyn_hair_short_bk_l_01(ComponentPose=@n20, Alpha=1, ExternalForce="0,0,0") {
    angularDampingOverride=0.9999
    angularSpringConstant=50
    bGravityOverrideInSimSpace=true
    bOverrideAngularDamping=true
    simulationSpace=BoneRelative
    alphaBoolBlend {blendOption=Linear}
    boundBone {boneName=dyn_hair_short_bk_l_01}
    componentAppliedLinearAccClamp=<2500 2500 2500>
    componentLinearAccScale=<0.7 0.7 0.8>
    componentLinearVelScale=<0.7 0.7 0.8>
    physicsBodyDefinitions [{
      boundBone {boneName=dyn_hair_short_bk_l_01}
      boxExtents=<10 10 10>
      constraintSetup {angularLimitsMax=<2 30 15>, angularLimitsMin=<-2 -15 -15>, angularTarget=<1 0 0>}
      localJointOffset=<1 0 0>
     }]
    relativeSpaceBone {boneName=root}
   }
   n22 = Anim Dynamics - Bone: dyn_hair_short_bk_l_02(ComponentPose=@n21, Alpha=1, ExternalForce="0,0,0") {
    angularDampingOverride=0.9999
    angularSpringConstant=50
    bGravityOverrideInSimSpace=true
    bOverrideAngularDamping=true
    simulationSpace=BoneRelative
    alphaBoolBlend {blendOption=Linear}
    boundBone {boneName=dyn_hair_short_bk_l_02}
    componentAppliedLinearAccClamp=<2500 2500 2500>
    componentLinearAccScale=<0.7 0.7 0.8>
    componentLinearVelScale=<0.7 0.7 0.8>
    physicsBodyDefinitions [{
      boundBone {boneName=dyn_hair_short_bk_l_02}
      boxExtents=<10 10 10>
      constraintSetup {angularLimitsMax=<2 30 15>, angularLimitsMin=<-2 -15 -15>, angularTarget=<1 0 0>}
      localJointOffset=<1 0 0>
     }]
    relativeSpaceBone {boneName=root}
   }
   n23 = Anim Dynamics - Bone: dyn_hair_short_side_l_01(ComponentPose=@n22, Alpha=1, ExternalForce="0,0,0") {
    angularDampingOverride=0.9999
    angularSpringConstant=50
    bGravityOverrideInSimSpace=true
    bOverrideAngularDamping=true
    simulationSpace=BoneRelative
    alphaBoolBlend {blendOption=Linear}
    boundBone {boneName=dyn_hair_short_side_l_01}
    componentAppliedLinearAccClamp=<2500 2500 2500>
    componentLinearAccScale=<0.7 0.7 0.8>
    componentLinearVelScale=<0.7 0.7 0.8>
    physicsBodyDefinitions [{
      boundBone {boneName=dyn_hair_short_side_l_01}
      boxExtents=<10 10 10>
      constraintSetup {angularLimitsMax=<1 10 10>, angularLimitsMin=<-1 -10 -10>, angularTarget=<1 0 0>}
      localJointOffset=<1 0 0>
     }]
    relativeSpaceBone {boneName=root}
   }
   n24 = Anim Dynamics - Bone: dyn_hair_short_side_r_01(ComponentPose=@n23, Alpha=1, ExternalForce="0,0,0") {
    angularDampingOverride=0.9999
    angularSpringConstant=50
    bGravityOverrideInSimSpace=true
    bOverrideAngularDamping=true
    simulationSpace=BoneRelative
    alphaBoolBlend {blendOption=Linear}
    boundBone {boneName=dyn_hair_short_side_r_01}
    componentAppliedLinearAccClamp=<2500 2500 2500>
    componentLinearAccScale=<0.7 0.7 0.8>
    componentLinearVelScale=<0.7 0.7 0.8>
    physicsBodyDefinitions [{
      boundBone {boneName=dyn_hair_short_side_r_01}
      boxExtents=<10 10 10>
      constraintSetup {angularLimitsMax=<1 10 10>, angularLimitsMin=<-1 -10 -10>, angularTarget=<1 0 0>}
      localJointOffset=<1 0 0>
     }]
    relativeSpaceBone {boneName=root}
   }
   n25 = Anim Dynamics - Bone: dyn_hair_short_side_l_02(ComponentPose=@n24, Alpha=1, ExternalForce="0,0,0") {
    angularDampingOverride=0.9999
    angularSpringConstant=50
    bGravityOverrideInSimSpace=true
    bOverrideAngularDamping=true
    simulationSpace=BoneRelative
    alphaBoolBlend {blendOption=Linear}
    boundBone {boneName=dyn_hair_short_side_l_02}
    componentAppliedLinearAccClamp=<2500 2500 2500>
    componentLinearAccScale=<0.7 0.7 0.8>
    componentLinearVelScale=<0.7 0.7 0.8>
    physicsBodyDefinitions [{
      boundBone {boneName=dyn_hair_short_side_l_02}
      boxExtents=<10 10 10>
      constraintSetup {angularLimitsMax=<1 10 10>, angularLimitsMin=<-1 -10 -10>, angularTarget=<1 0 0>}
      localJointOffset=<1 0 0>
     }]
    relativeSpaceBone {boneName=root}
   }
   n26 = Anim Dynamics - Bone: dyn_hair_short_side_r_02(ComponentPose=@n25, Alpha=1, ExternalForce="0,0,0") {
    angularDampingOverride=0.9999
    angularSpringConstant=50
    bGravityOverrideInSimSpace=true
    bOverrideAngularDamping=true
    simulationSpace=BoneRelative
    alphaBoolBlend {blendOption=Linear}
    boundBone {boneName=dyn_hair_short_side_r_02}
    componentAppliedLinearAccClamp=<2500 2500 2500>
    componentLinearAccScale=<0.7 0.7 0.8>
    componentLinearVelScale=<0.7 0.7 0.8>
    physicsBodyDefinitions [{
      boundBone {boneName=dyn_hair_short_side_r_02}
      boxExtents=<10 10 10>
      constraintSetup {angularLimitsMax=<1 10 10>, angularLimitsMin=<-1 -10 -10>, angularTarget=<1 0 0>}
      localJointOffset=<1 0 0>
     }]
    relativeSpaceBone {boneName=root}
   }
   n27 = Get Skydiving() -> (Skydiving:double)
   n28 = "Transform (Modify) Bone - Bone: dyn_crown_short_mstr"(ComponentPose=@n26, Alpha=@n27, Translation="0,0,0", Rotation="15,0,0", Scale="1,1,1") {
    rotationMode=BMM_Additive
    rotationSpace=BCS_BoneSpace
    translationMode=BMM_Additive
    translationSpace=BCS_BoneSpace
    alphaBoolBlend {blendOption=Linear}
    alphaScaleBiasClamp {bInterpResult=true, interpSpeedDecreasing=7, interpSpeedIncreasing=5}
    boneToModify {boneName=dyn_crown_short_mstr}
    rotation {pitch=15}
   }
   n29 = "Transform (Modify) Bone - Bone: dyn_hair_short_bk_r_mstr"(ComponentPose=@n28, Alpha=@n27, Translation="2,0,5", Rotation="15,0,0", Scale="1,1,1") {
    rotationMode=BMM_Additive
    rotationSpace=BCS_BoneSpace
    translationMode=BMM_Additive
    translationSpace=BCS_BoneSpace
    alphaBoolBlend {blendOption=Linear}
    alphaScaleBiasClamp {bInterpResult=true, interpSpeedDecreasing=7, interpSpeedIncreasing=5}
    boneToModify {boneName=dyn_hair_short_bk_r_mstr}
    rotation {pitch=15}
    translation {x=2, z=5}
   }
   n30 = "Transform (Modify) Bone - Bone: dyn_hair_short_side_r_mstr"(ComponentPose=@n29, Alpha=@n27, Translation="0,0,2", Rotation="15,0,0", Scale="1,1,1") {
    rotationMode=BMM_Additive
    rotationSpace=BCS_BoneSpace
    translationMode=BMM_Additive
    translationSpace=BCS_BoneSpace
    alphaBoolBlend {blendOption=Linear}
    alphaScaleBiasClamp {bInterpResult=true, interpSpeedDecreasing=7, interpSpeedIncreasing=5}
    boneToModify {boneName=dyn_hair_short_side_r_mstr}
    rotation {pitch=15}
    translation {z=2}
   }
   n31 = Control Rig!(Source=Component To Local(ComponentPose=@n30), Alpha, Crouch:bool=Get bIsCrouching()) -> (Pose) {
    alphaBoolBlend {blendOption=Linear}
    controlRigAssetReference {blueprintRigClass=&/F_MED_SuperNovaTaro_FaceAcc_CtrlRig.F_MED_SuperNovaTaro_FaceAcc_CtrlRig_C}
   }
   Output Pose(Result=@n31) {layerGroup=None}
  ]}
 EventGraph {nodes [
   Event BlueprintUpdateAnimation()
   > n1 = Sequence()
   > n2 = Gravity Override State Function DEPRECATED(CharacterPartAnimInstance=Self-Reference(), GameTimeInSeconds=GetGameTimeInSeconds(), GravityOverrideParams=Get Gravity Override Params())
   > Set GravityOverride(GravityOverride=@n2)
   Set Skydiving(execute=@n1.then_1, Skydiving:double="To Float (Boolean)"(InBool=Get bIsSkyDiving())) -> (Output_Get:double)
   n3 = Get bIsTargeting()
   n4 = Get bIsPlayingMeleeAnim()
   > Set TargetOrMelee(TargetOrMelee:double="To Float (Boolean)"(InBool=OR Boolean(A=@n3, B=@n4))) -> (Output_Get:double)
   n5 = Event OnBeginFalling()
   Set jumping(execute=[@n5.then, @n7.then], jumping:double=1) -> (Output_Get:double)
   Event BlueprintInitializeAnimation()
   > Set Skel_Mesh(Skel_Mesh=GetOwnerSkeletalMeshComponent())
   n6 = Event OnEndFalling()
   Set jumping(execute=[@n6.then, @n8.then], jumping=0)
   n7 = Event OnBeginJumping()
   Event OnBeginSkydiving()
   > Set Skydiving(Skydiving=1)
   n8 = Event OnEndJumping()
   Event OnEndSkydiving()
   > Set Skydiving(Skydiving=0)
  ]}
]
variables [
 Gravity Override Params {type=GravityOverrideParamsStruct}
 GravityOverride {type=Vector}
 Jumping {type=number}
 Skel_Mesh {type=SkeletalMeshComponent}
 Skydiving {type=number}
 TargetOrMelee {type=number}
]
